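# Deformation Control Rig for the M_LRG_Invader body that drives the left and right shoulder armor bones.
# Each armor master bone is parent-constrained between clavicle and upper-arm twist, favoring the clavicle.
# Remapped arm angles weight additive corrective offsets, gated by skydivingINV, or aim and offset corrections, gated by skydiving.
# The final right-side corrective is Kalman-smoothed; spare AimItem and TransformConstraint nodes are unconnected, and pistol is unused.
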ControlRigBlueprint /BRCosmetics/Characters/Player/Male/Large/Bodies/M_LRG_Invader/Meshes/M_LRG_Invader_Deform_CtrlRig_armor
graphs [
 RigVMFunctionLibrary {links [], nodes []}
 RigVMModel {nodes [
   RigUnit_BeginExecution = &RigUnit_BeginExecution()
   &RigUnit_ParentConstraint(
    AdvancedSettings=(RotationOrderForFilter=XZY)
    Child=(Type=Bone,Name="dyn_armor_r_MSTR")
    Parents=((Item=(Type=Bone,Name="clavicle_r"),Weight=0.8),(Item=(Type=Bone,Name="upperarm_twist_01_r"),Weight=0.2))
    Weight=1
    ExecutePin=@RigUnit_BeginExecution.ExecutePin
   )
   n1 = /Script/RigVM.RigVMFunction_MathFloatAbs(Value=&RigVMFunction_MathQuaternionToEuler(RotationOrder=ZYX, Value=&RigUnit_GetRelativeBoneTransform(Bone=upperarm_r, Space=spine_05).Transform.Rotation).Result.Z) {is=struct RigVMUnitNode, template="Absolute::Execute(in Value,out Result)"}
   n2 = RigVMVariableNode() {variable=skydivingINV}
   n3 = &RigVMFunction_MathQuaternionToEuler(
    RotationOrder=ZYX
    Value=/Script/ControlRig.RigUnit_GetInitialBoneTransform(Bone=dyn_armor_r, Space=LocalSpace) {is=struct RigVMUnitNode}.Transform.Rotation
   )
   > /Script/ControlRig.RigUnit_ModifyBoneTransforms(
    BoneToModify=((Bone="dyn_armor_r",Transform=(Rotation=(X=0,Y=0,Z=0,W=1),Translation=(X=-3,Y=-3,Z=0),Scale3D=(X=1,Y=1,Z=1))))
    Mode=AdditiveLocal
    WeightMaximum=1
    Weight=&RigVMFunction_MathFloatMul(
     A=&RigVMFunction_MathFloatRemap(SourceMaximum=55, SourceMinimum=0, TargetMinimum=0, bClamp=true, Value=@n1.Result).Result
     B=@n2.Value
    ) {template="Multiply::Execute(in A,in B,out Result)"}.Result
    BoneToModify.0.Transform.Rotation=&RigVMFunction_MathQuaternionFromEuler(Euler=/Script/RigVM.RigVMFunction_MathVectorAdd(B=(X=-15,Y=10,Z=-20), A=@n3.Result) {is=struct RigVMUnitNode, template="Add::Execute(in A,in B,out Result)"}.Result).Result
   ) {is=struct RigVMUnitNode}
   n4 = &RigVMFunction_MathQuaternionToEuler(RotationOrder=ZYX, Value=&RigUnit_GetInitialBoneTransform(Bone=dyn_armor_r, Space=LocalSpace).Transform.Rotation)
   > &RigUnit_ModifyBoneTransforms(
    BoneToModify=((Bone="dyn_armor_r",Transform=(Rotation=(X=0,Y=0,Z=0,W=1),Translation=(X=0,Y=0,Z=2),Scale3D=(X=1,Y=1,Z=1))))
    Mode=AdditiveLocal
    WeightMaximum=1
    WeightMinimum=0
    Weight=&RigVMFunction_MathFloatMul(
     A=&RigVMFunction_MathFloatRemap(SourceMaximum=130, SourceMinimum=50, TargetMinimum=0, bClamp=true, Value=@n1.Result).Result
     B=@n2.Value
    ) {template="Multiply::Execute(in A,in B,out Result)"}.Result
    BoneToModify.0.Transform.Rotation=&RigVMFunction_MathQuaternionFromEuler(Euler=&RigVMFunction_MathVectorAdd(B=(X=-35,Y=-10,Z=-20), A=@n4.Result) {template="Add::Execute(in A,in B,out Result)"}.Result).Result
   )
   n5 = &RigVMFunction_MathFloatAbs(Value=&RigVMFunction_MathQuaternionToEuler(RotationOrder=YZX, Value=&RigUnit_GetRelativeBoneTransform(Bone=upperarm_r, Space=spine_05).Transform.Rotation).Result.Y) {template="Absolute::Execute(in Value,out Result)"}
   n6 = RigVMVariableNode() {variable=skydiving}
   > /Script/ControlRig.RigUnit_AimBone(
    Bone=dyn_armor_r_MSTR
    DebugSettings=(Scale=10,WorldOffset=(Rotation=(X=0,Y=0,Z=0,W=1),Translation=(X=0,Y=0,Z=0),Scale3D=(X=1,Y=1,Z=1)))
    Primary=(Weight=1,Axis=(X=1,Y=0,Z=0),Target=(X=1,Y=0,Z=0),Kind=Location)
    Secondary=(Weight=0,Axis=(X=0,Y=0,Z=0),Target=(X=1,Y=0,Z=0),Kind=Location)
    bPropagateToChildren=True
    Weight=&RigVMFunction_MathFloatMul(
     A=&RigVMFunction_MathFloatRemap(SourceMaximum=120, SourceMinimum=50, TargetMinimum=0, bClamp=true, Value=@n5.Result).Result
     B=@n6.Value
    ) {template="Multiply::Execute(in A,in B,out Result)"}.Result
    Primary.Target=&RigUnit_GetBoneTransform(Bone=dyn_armor_r_AIM).Transform.Translation
   ) {is=struct RigVMUnitNode}
   n7 = &RigVMFunction_MathFloatAbs(Value=&RigVMFunction_MathQuaternionToEuler(RotationOrder=YZX, Value=&RigUnit_GetRelativeBoneTransform(Bone=upperarm_r, Space=spine_05).Transform.Rotation).Result.Y) {template="Absolute::Execute(in Value,out Result)"}
   > &RigUnit_ModifyBoneTransforms(
    BoneToModify=((Bone="dyn_armor_r",Transform=(Rotation=(X=0,Y=0,Z=0,W=1),Translation=(X=5,Y=0,Z=1),Scale3D=(X=1,Y=1,Z=1))))
    Mode=AdditiveLocal
    WeightMaximum=1
    Weight=&RigVMFunction_MathFloatMul(
     B=&RigVMFunction_MathFloatRemap(SourceMaximum=170, SourceMinimum=50, TargetMinimum=0, bClamp=true, Value=@n7.Result).Result
     A=@n6.Value
    ) {template="Multiply::Execute(in A,in B,out Result)"}.Result
   )
   > &RigUnit_ParentConstraint(
    AdvancedSettings=(RotationOrderForFilter=XZY)
    Child=(Type=Bone,Name="dyn_armor_l_MSTR")
    Parents=((Item=(Type=Bone,Name="clavicle_l"),Weight=0.8),(Item=(Type=Bone,Name="upperarm_twist_01_l"),Weight=0.2))
    Weight=1
   )
   n8 = &RigVMFunction_MathFloatAbs(Value=&RigVMFunction_MathQuaternionToEuler(RotationOrder=ZYX, Value=&RigUnit_GetRelativeBoneTransform(Bone=upperarm_l, Space=spine_05).Transform.Rotation).Result.Z) {template="Absolute::Execute(in Value,out Result)"}
   n9 = &RigVMFunction_MathQuaternionToEuler(RotationOrder=ZYX, Value=&RigUnit_GetInitialBoneTransform(Bone=dyn_armor_l, Space=LocalSpace).Transform.Rotation)
   > &RigUnit_ModifyBoneTransforms(
    BoneToModify=((Bone="dyn_armor_l",Transform=(Rotation=(X=0,Y=0,Z=0,W=1),Translation=(X=1,Y=0,Z=0),Scale3D=(X=1,Y=1,Z=1))))
    Mode=AdditiveLocal
    WeightMaximum=1
    WeightMinimum=0
    Weight=&RigVMFunction_MathFloatMul(
     A=&RigVMFunction_MathFloatRemap(SourceMaximum=60, SourceMinimum=150, TargetMinimum=0, bClamp=true, Value=@n8.Result).Result
     B=RigVMVariableNode() {variable=skydivingINV}.Value
    ) {template="Multiply::Execute(in A,in B,out Result)"}.Result
    BoneToModify.0.Transform.Rotation=&RigVMFunction_MathQuaternionFromEuler(Euler=&RigVMFunction_MathVectorAdd(B=(X=40,Y=0,Z=10), A=@n9.Result) {template="Add::Execute(in A,in B,out Result)"}.Result).Result
   )
   n10 = &RigVMFunction_MathFloatAbs(Value=&RigVMFunction_MathQuaternionToEuler(RotationOrder=YZX, Value=&RigUnit_GetRelativeBoneTransform(Bone=upperarm_l, Space=spine_05).Transform.Rotation).Result.Y) {template="Absolute::Execute(in Value,out Result)"}
   > &RigUnit_AimBone(
    Bone=dyn_armor_l_MSTR
    DebugSettings=(Scale=10,WorldOffset=(Rotation=(X=0,Y=0,Z=0,W=1),Translation=(X=0,Y=0,Z=0),Scale3D=(X=1,Y=1,Z=1)))
    Primary=(Weight=1,Axis=(X=1,Y=0,Z=0),Target=(X=1,Y=0,Z=0),Kind=Location)
    Secondary=(Weight=0,Axis=(X=0,Y=0,Z=0),Target=(X=1,Y=0,Z=0),Kind=Location)
    bPropagateToChildren=True
    Weight=&RigVMFunction_MathFloatMul(
     A=&RigVMFunction_MathFloatRemap(SourceMaximum=40, SourceMinimum=150, TargetMinimum=0, bClamp=true, Value=@n10.Result).Result
     B=RigVMVariableNode() {variable=skydiving}.Value
    ) {template="Multiply::Execute(in A,in B,out Result)"}.Result
    Primary.Target=&RigUnit_GetBoneTransform(Bone=dyn_armor_l_AIM).Transform.Translation
   )
   n11 = &RigVMFunction_MathFloatAbs(Value=&RigVMFunction_MathQuaternionToEuler(RotationOrder=YZX, Value=&RigUnit_GetRelativeBoneTransform(Bone=upperarm_l, Space=spine_05).Transform.Rotation).Result.Y) {template="Absolute::Execute(in Value,out Result)"}
   > &RigUnit_ModifyBoneTransforms(
    BoneToModify=((Bone="dyn_armor_l",Transform=(Rotation=(X=0,Y=0,Z=0,W=1),Translation=(X=5,Y=0,Z=1),Scale3D=(X=1,Y=1,Z=1))))
    Mode=AdditiveLocal
    WeightMaximum=1
    Weight=&RigVMFunction_MathFloatMul(
     A=&RigVMFunction_MathFloatRemap(SourceMaximum=5, SourceMinimum=150, TargetMinimum=0, bClamp=true, Value=@n11.Result).Result
     B=RigVMVariableNode() {variable=skydiving}.Value
    ) {template="Multiply::Execute(in A,in B,out Result)"}.Result
   )
   n12 = &RigVMFunction_MathFloatAbs(Value=&RigVMFunction_MathQuaternionToEuler(RotationOrder=YZX, Value=&RigUnit_GetRelativeBoneTransform(Bone=clavicle_l, Space=upperarm_l).Transform.Rotation).Result.Z) {template="Absolute::Execute(in Value,out Result)"}
   n13 = &RigVMFunction_MathQuaternionToEuler(RotationOrder=ZYX, Value=&RigUnit_GetInitialBoneTransform(Bone=dyn_armor_l, Space=LocalSpace).Transform.Rotation)
   > &RigUnit_ModifyBoneTransforms(
    BoneToModify=((Bone="dyn_armor_l",Transform=(Rotation=(X=0,Y=0,Z=0,W=1),Translation=(X=0,Y=0,Z=0),Scale3D=(X=1,Y=1,Z=1))))
    Mode=AdditiveLocal
    WeightMaximum=1
    WeightMinimum=0
    Weight=&RigVMFunction_MathFloatMul(
     A=&RigVMFunction_MathFloatRemap(SourceMaximum=30, SourceMinimum=10, TargetMinimum=0, bClamp=true, Value=@n12.Result).Result
     B=RigVMVariableNode() {variable=skydivingINV}.Value
    ) {template="Multiply::Execute(in A,in B,out Result)"}.Result
    BoneToModify.0.Transform.Rotation=&RigVMFunction_MathQuaternionFromEuler(Euler=&RigVMFunction_MathVectorAdd(B=(X=15,Y=0,Z=10), A=@n13.Result) {template="Add::Execute(in A,in B,out Result)"}.Result).Result
   )
   n14 = &RigUnit_GetRelativeBoneTransform(Bone=clavicle_r, Space=upperarm_r)
   n15 = &RigVMFunction_MathFloatRemap(
    SourceMaximum=45
    SourceMinimum=30
    TargetMinimum=0
    bClamp=true
    Value=&RigVMFunction_MathFloatAbs(Value=&RigVMFunction_MathQuaternionToEuler(RotationOrder=YZX, Value=@n14.Transform.Rotation).Result.Z) {template="Absolute::Execute(in Value,out Result)"}.Result
   )
   n16 = RigVMVariableNode() {variable=skydivingINV}
   n17 = &RigVMFunction_MathQuaternionToEuler(RotationOrder=ZYX, Value=&RigUnit_GetInitialBoneTransform(Bone=dyn_armor_r, Space=LocalSpace).Transform.Rotation)
   > &RigUnit_ModifyBoneTransforms(
    BoneToModify=((Bone="dyn_armor_r",Transform=(Rotation=(X=0,Y=0,Z=0,W=1),Translation=(X=0,Y=0,Z=0),Scale3D=(X=1,Y=1,Z=1))))
    Mode=AdditiveLocal
    WeightMaximum=1
    WeightMinimum=0
    Weight=&RigVMFunction_KalmanFloat(
     BufferSize=2
     Value=&RigVMFunction_MathFloatMul(A=@n15.Result, B=@n16.Value) {template="Multiply::Execute(in A,in B,out Result)"}.Result
    ) {template="KalmanFilter::Execute(in Value,in BufferSize,out Result)"}.Result
    BoneToModify.0.Transform.Rotation=&RigVMFunction_MathQuaternionFromEuler(Euler=&RigVMFunction_MathVectorAdd(B=(X=-15,Y=0,Z=-15), A=@n17.Result) {template="Add::Execute(in A,in B,out Result)"}.Result).Result
   )
   /Script/ControlRig.RigUnit_AimItem(
    DebugSettings=(Scale=10,WorldOffset=(Rotation=(X=0,Y=0,Z=0,W=1),Translation=(X=0,Y=0,Z=0),Scale3D=(X=1,Y=1,Z=1)))
    ExecutePin=()
    Item=(Type=Bone,Name="dyn_armor_r_MSTR")
    Primary=(Weight=1,Axis=(X=1,Y=0,Z=0),Target=(X=1,Y=0,Z=0),Kind=Location,Space=(Type=Bone))
    Secondary=(Weight=0,Axis=(X=0,Y=0,Z=1),Target=(X=1,Y=0,Z=0),Kind=Location,Space=(Type=Bone))
    Weight=1
    Primary.Target=&RigUnit_GetTransform(Item=(Type=Bone,Name="dyn_armor_r_AIM")).Transform.Translation
   ) {is=struct RigVMUnitNode}
   RigUnit_GetTransform = &RigUnit_GetTransform(Item=(Type=Bone,Name="upperarm_twist_01_r"))
   /Script/ControlRig.RigUnit_TransformConstraintPerItem(
    BaseItem=(Type=Bone)
    BaseTransformSpace=GlobalSpace
    ExecutePin=()
    Item=(Type=Bone,Name="dyn_armor_r_MSTR")
    Targets="((Transform=(Rotation=(X=0,Y=0,Z=0,W=1),Translation=(X=0,Y=0,Z=0),Scale3D=(X=1,Y=1,Z=1)),Weight=0.8,bMaintainOffset=True,Filter=(TranslationFilter=(bX=True,bY=True,bZ=True),RotationFilter=(bX=True,bY=True,bZ=True),ScaleFilter=(bX=True,bY=True,bZ=True))),(T ...(truncated)"
    bUseInitialTransforms=True
    Targets.1.Transform=@RigUnit_GetTransform.Transform
    Targets.0.Transform=&RigUnit_GetTransform(Item=(Type=Bone,Name="clavicle_r")).Transform
    BaseTransform=&RigUnit_GetTransform(Item=(Type=Bone,Name="dyn_armor_r_MSTR")).Transform
   ) {is=struct RigVMUnitNode}
   &RigUnit_TransformConstraintPerItem(
    BaseItem=(Type=Bone)
    BaseTransformSpace=GlobalSpace
    ExecutePin=()
    Item=(Type=Bone,Name="dyn_armor_l_MSTR")
    Targets="((Transform=(Rotation=(X=0,Y=0,Z=0,W=1),Translation=(X=0,Y=0,Z=0),Scale3D=(X=1,Y=1,Z=1)),Weight=0.8,bMaintainOffset=True,Filter=(TranslationFilter=(bX=True,bY=True,bZ=True),RotationFilter=(bX=True,bY=True,bZ=True),ScaleFilter=(bX=True,bY=True,bZ=True))),(T ...(truncated)"
    bUseInitialTransforms=True
    Targets.0.Transform=&RigUnit_GetTransform(Item=(Type=Bone,Name="clavicle_l")).Transform
    BaseTransform=&RigUnit_GetTransform(Item=(Type=Bone,Name="dyn_armor_l_MSTR")).Transform
    Targets.1.Transform=&RigUnit_GetTransform(Item=(Type=Bone,Name="upperarm_twist_01_l")).Transform
   )
   RigVMVariableNode() {variable=skydiving}
  ]}
]
hierarchy [
 root {loc=<0 0 0>, rot=<0 0 0 1>}
 pelvis {parent=root, loc=<0 -0.7321 92.4218>, rot=<0.0127 0.707 -0.0127 -0.707>}
 spine_01 {parent=pelvis, loc=<5.6002 -0.5128 0>, rot=<0 0 0.034 -0.9994>}
 spine_02 {parent=spine_01, loc=<8.0753 0.0139 0>, rot=<0 0 0.0238 -0.9997>}
 spine_03 {parent=spine_02, loc=<7.0393 0 0>, rot=<0 0 0.0149 -0.9999>}
 spine_04 {parent=spine_03, loc=<8.1024 0 0>, rot=<0 0 -0.0132 -0.9999>}
 spine_05 {parent=spine_04, loc=<10.0146 0 0>, rot=<0 0 -0.0015 -1>}
 clavicle_l {parent=spine_05, loc=<2.5856 -0.2866 -4.3777>, rot=<0.0465 -0.7289 -0.009 -0.6829>}
 upperarm_l {parent=clavicle_l, loc=<18.2264 0.1342 -0.001>, rot=<0.0253 -0.2693 0.0013 -0.9627>}
 lowerarm_l {parent=upperarm_l, loc=<29.0071 0.1686 -1.239>, rot=<0.0391 -0.0099 -0.2453 -0.9686>}
 hand_l {parent=lowerarm_l, loc=<25.8607 0.222 1.0274>, rot=<-0.7193 -0.0988 0.1193 -0.6773>}
 index_metacarpal_l {parent=hand_l, loc=<1.6535 0.0183 -2.96>, rot=<0.0158 -0.0124 -0.0173 -0.9996>}
 index_01_l {parent=index_metacarpal_l, loc=<8.3008 0.0357 -0.0524>, rot=<-0.0044 0.0485 0.0924 -0.9945>}
 index_02_l {parent=index_01_l, loc=<3.3587 -0.3004 0.0452>, rot=<-0.002 0.0027 0.3225 -0.9466>}
 index_03_l {parent=index_02_l, loc=<2.5578 -0.123 -0.0252>, rot=<-0.002 0.0067 0.2558 -0.9667>}
 pinky_metacarpal_l {parent=hand_l, loc=<1.7616 -0.9753 3.3814>, rot=<-0.0397 0.0814 -0.069 -0.9935>}
 pinky_01_l {parent=pinky_metacarpal_l, loc=<7.1709 -0.4831 -0.1094>, rot=<-0.0654 -0.0035 0.2242 -0.9723>}
 pinky_02_l {parent=pinky_01_l, loc=<2.8274 0.0893 -0.1043>, rot=<0.0037 0.0011 0.2854 -0.9584>}
 pinky_03_l {parent=pinky_02_l, loc=<2.2046 0.0307 -0.0078>, rot=<-0.0002 0.0098 0.1445 -0.9895>}
 thumb_01_l {parent=hand_l, loc=<1.3118 -0.3191 -2.459>, rot=<0.6546 -0.4042 0.0664 -0.6354>}
 thumb_02_l {parent=thumb_01_l, loc=<5.9336 -0.0448 0.0401>, rot=<-0.0037 0.0213 0.298 -0.9543>}
 thumb_03_l {parent=thumb_02_l, loc=<2.2223 0.3682 -0.0907>, rot=<0.0118 -0.013 0.0066 -0.9998>}
 lowerarm_twist_01_l {parent=lowerarm_l, loc=<17.8549 0 0>, rot=<0 0 0 -1>}
 upperarm_twist_01_l {parent=upperarm_l, loc=<7.9796 0.0843 -0.6195>, rot=<0 0 0 1>}
 dyn_armor_l_MSTR {parent=upperarm_twist_01_l, loc=<-6.1364 -3.214 2.1434>, rot=<0 0 0 1>}
 dyn_armor_l {parent=dyn_armor_l_MSTR, loc=<0 0 0>, rot=<0 0 0 1>}
 upperarm_twist_02_l {parent=upperarm_l, loc=<19.5059 0.0843 -0.6195>, rot=<0 0 0 1>}
 dyn_armor_l_AIM {parent=upperarm_twist_02_l, loc=<12.3327 -3.214 2.142>, rot=<0 0 0 -1>}
 clavicle_r {parent=spine_05, loc=<2.5856 -0.2865 4.3777>, rot=<0.7289 0.0465 -0.6829 0.009>}
 upperarm_r {parent=clavicle_r, loc=<-18.2264 -0.1342 0.001>, rot=<0.0253 -0.2693 0.0013 -0.9627>}
 lowerarm_r {parent=upperarm_r, loc=<-29.0072 -0.1686 1.239>, rot=<0.0391 -0.0099 -0.2453 -0.9686>}
 hand_r {parent=lowerarm_r, loc=<-25.8608 -0.222 -1.0274>, rot=<-0.7193 -0.0988 0.1193 -0.6773>}
 index_metacarpal_r {parent=hand_r, loc=<-1.6534 -0.0181 2.9601>, rot=<0.0158 -0.0124 -0.0173 -0.9996>}
 index_01_r {parent=index_metacarpal_r, loc=<-8.3008 -0.0357 0.0524>, rot=<-0.0044 0.0485 0.0924 -0.9945>}
 index_02_r {parent=index_01_r, loc=<-3.3589 0.3003 -0.0453>, rot=<-0.002 0.0027 0.3225 -0.9466>}
 index_03_r {parent=index_02_r, loc=<-2.5578 0.123 0.0252>, rot=<-0.002 0.0067 0.2558 -0.9667>}
 pinky_metacarpal_r {parent=hand_r, loc=<-1.7617 0.9753 -3.3814>, rot=<-0.0397 0.0814 -0.069 -0.9935>}
 pinky_01_r {parent=pinky_metacarpal_r, loc=<-7.1707 0.4832 0.1094>, rot=<-0.0654 -0.0035 0.2242 -0.9723>}
 pinky_02_r {parent=pinky_01_r, loc=<-2.8274 -0.0892 0.1043>, rot=<0.0037 0.0011 0.2854 -0.9584>}
 pinky_03_r {parent=pinky_02_r, loc=<-2.2045 -0.031 0.0078>, rot=<-0.0002 0.0098 0.1445 -0.9895>}
 ring_metacarpal_r {parent=hand_r, loc=<-2.2073 0.3627 -1.3638>, rot=<-0.0284 0.0722 -0.0458 -0.9959>}
 thumb_01_r {parent=hand_r, loc=<-1.3119 0.3192 2.459>, rot=<0.6546 -0.4042 0.0664 -0.6354>}
 thumb_02_r {parent=thumb_01_r, loc=<-5.9336 0.0447 -0.0401>, rot=<-0.0037 0.0213 0.298 -0.9543>}
 thumb_03_r {parent=thumb_02_r, loc=<-2.2223 -0.3682 0.0907>, rot=<0.0118 -0.013 0.0066 -0.9998>}
 lowerarm_twist_01_r {parent=lowerarm_r, loc=<-17.855 0 0>, rot=<0 0 0 -1>}
 lowerarm_twist_02_r {parent=lowerarm_r, loc=<-9.6199 0 0>, rot=<0 0 0 -1>}
 upperarm_twist_01_r {parent=upperarm_r, loc=<-7.9796 -0.0843 0.6195>, rot=<0 0 0 1>}
 dyn_armor_r_MSTR {parent=upperarm_twist_01_r, loc=<6.136 3.214 -2.143>, rot=<0 -1 0 0>}
 dyn_armor_r {parent=dyn_armor_r_MSTR, loc=<0 0 0>, rot=<0 0 0 1>}
 upperarm_twist_02_r {parent=upperarm_r, loc=<-19.506 -0.0843 0.6195>, rot=<0 0 0 1>}
 dyn_armor_r_AIM {parent=upperarm_twist_02_r, loc=<-12.3376 3.214 -2.143>, rot=<0 -1 0 0>}
 neck_01 {parent=spine_05, loc=<10.815 1.9196 0>, rot=<0 0 -0.1864 -0.9825>}
 neck_02 {parent=neck_01, loc=<4.7347 0.3045 0>, rot=<0 0 0.0363 -0.9993>}
 head {parent=neck_02, loc=<5.7654 0.5554 -0.0001>, rot=<0.0678 -0.009 0.1647 -0.984>}
 head_offset {parent=neck_02, loc=<5.7632 0.5552 0>, rot=<0 0 0.115 -0.9934>}
 dyn_tube_MSTR {parent=spine_05, loc=<-9.2935 -19.0863 -4.5103>, rot=<0.2942 -0.6302 0.6655 0.271>}
 dyn_tube {parent=dyn_tube_MSTR, loc=<0 0 0>, rot=<0 0 0 -1>}
 thigh_l {parent=pelvis, loc=<-1.8716 -0.0109 -11.9936>, rot=<0.0012 0.0572 0.0155 -0.9982>}
 calf_l {parent=thigh_l, loc=<-41.9178 -0.2246 0.0926>, rot=<0.0006 -0.0161 -0.0391 -0.9991>}
 calf_twist_02_l {parent=calf_l, loc=<-15.9268 -0.0967 0.0397>, rot=<0.0237 0.0095 -0.0035 -0.9997>}
 foot_l {parent=calf_l, loc=<-38.7057 1.2438 0.2453>, rot=<0.0975 -0.0474 0.0317 -0.9936>}
 ball_l {parent=foot_l, loc=<-6.1614 13.1469 0.6807>, rot=<0.0001 -0.0001 -0.7063 -0.7079>}
 thigh_twist_01_l {parent=thigh_l, loc=<-21.6303 0 0>, rot=<0.0013 0 0 -1>}
 thigh_r {parent=pelvis, loc=<-1.8714 -0.0109 11.9936>, rot=<0.0572 -0.0012 0.9982 0.0155>}
 calf_r {parent=thigh_r, loc=<41.918 0.2246 -0.0926>, rot=<0.0006 -0.0161 -0.0391 -0.9991>}
 calf_twist_02_r {parent=calf_r, loc=<15.9268 0.0967 -0.0397>, rot=<0.0237 0.0095 -0.0035 -0.9997>}
 foot_r {parent=calf_r, loc=<38.7058 -1.2438 -0.2453>, rot=<0.0975 -0.0474 0.0317 -0.9936>}
 ball_r {parent=foot_r, loc=<6.1614 -13.1469 -0.6807>, rot=<0.0001 -0.0001 -0.7063 -0.7079>}
 thigh_twist_01_r {parent=thigh_r, loc=<21.6303 0 0>, rot=<0.0013 0 0 -1>}
]
variables [
 pistol {type=double, default=0}
 skydiving {type=double, default=0}
 skydivingINV {type=double, default=0}
]
