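# Physics asset for the dynamic tube attachment on the M_LRG_Invader large male cosmetic body mesh.
# It has two capsule bodies, one on the dyn_tube bone and a tiny one on the dyn_tube_MSTR bone.
# A single constraint links the tube to the master bone with narrow swing and tighter twist limits, restricting how far the tube can move relative to it.
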PhysicsAsset /BRCosmetics/Characters/Player/Male/Large/Bodies/M_LRG_Invader/Meshes/M_LRG_Invader_dyn_tube
bodies {
 dyn_tube [{shapeType=Capsule, length=10, radius=4, center=<5 0 0>, rotation=<rot 90 0 0>}]
 dyn_tube_MSTR [{shapeType=Capsule, length=0.505, radius=0.505, rotation=<rot 0 -90 0>}]
}
constraints [{bone1Name=dyn_tube, bone2Name=dyn_tube_MSTR, swing1LimitDegrees=20, swing2LimitDegrees=20, twistLimitDegrees=5}]
