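# Post-process style cosmetic AnimBP for the M_LRG_Invader body that copies the attached parent pose and corrects armor, head and tube deformation.
# A Control Rig feeds additive bone offsets, then RigidBody simulates dynamics.
# EventGraph derives alphas from bone deltas, skydiving, pistol state and PSD_Func, a cone-angle pose-space driver for right-arm armor.
# A visible Invader-tagged pickaxe toggles a material hide parameter and zeroes Control Rig and arm-armor alphas.
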AnimBlueprint /BRCosmetics/Characters/Player/Male/Large/Bodies/M_LRG_Invader/Meshes/M_LRG_Invader_AnimBP
target_skeleton=/Game/Characters/Player/Male/Large/Bodies/M_LRG_Invader/Meshes/M_LRG_Invader_Skeleton
graphs [
 AnimGraph {nodes [
   n1 = Control Rig!(
    Source=Copy Pose From Mesh() {bUseAttachedParent=true}
    Alpha=Get ctrl_rig_alpha() -> (ctrl_rig_alpha:double)
    skydivingINV:double=Get isSkydivingINV() -> (isSkydivingINV:double)
    skydiving:double=Get isSkydiving() -> (isSkydiving:double)
    pistol:double=Get pistolAimAlpha() -> (pistolAimAlpha:double)
   ) -> (Pose) {
    alphaBoolBlend {blendOption=Linear}
    controlRigAssetReference {blueprintRigClass=&/M_LRG_Invader_Deform_CtrlRig_armor.M_LRG_Invader_Deform_CtrlRig_armor_C}
   }
   n2 = "Transform (Modify) Bone - Bone: dyn_tube_MSTR"(
    ComponentPose=Local To Component(LocalPose=@n1)
    Alpha=Get spine_twist() -> (spine_twist:double)
    Translation="0,0,0"
    Rotation="20,0,0"
    Scale="1,1,1"
   ) {
    rotationMode=BMM_Additive
    rotationSpace=BCS_BoneSpace
    alphaBoolBlend {blendOption=Linear}
    boneToModify {boneName=dyn_tube_MSTR}
    rotation {pitch=20}
   }
   n3 = "Transform (Modify) Bone - Bone: dyn_tube_MSTR"(
    ComponentPose=@n2
    Alpha=Get spine_rot_bk() -> (spine_rot_bk:double)
    Translation="-0.531,-0.555,1.025"
    Rotation="29.243,-16.989,-5.823"
    Scale="1,1,1"
   ) {
    rotationMode=BMM_Additive
    rotationSpace=BCS_BoneSpace
    translationMode=BMM_Additive
    translationSpace=BCS_BoneSpace
    alphaBoolBlend {blendOption=Linear}
    boneToModify {boneName=dyn_tube_MSTR}
    rotation=<rot 29.2429 -16.9893 -5.8231>
    translation=<-0.531 -0.555 1.025>
   }
   n4 = "Transform (Modify) Bone - Bone: head_offset"(ComponentPose=@n3, Alpha=Get look_left() -> (look_left:double), Translation="0,0,0", Rotation="0,0,40", Scale="1,1,1") {
    rotationMode=BMM_Additive
    rotationSpace=BCS_BoneSpace
    alphaBoolBlend {blendOption=Linear}
    boneToModify {boneName=head_offset}
    rotation {roll=40}
   }
   n5 = "Transform (Modify) Bone - Bone: head_offset"(ComponentPose=@n4, Alpha=Get look_right() -> (look_right:double), Translation="0,0,0", Rotation="0,0,-40", Scale="1,1,1") {
    rotationMode=BMM_Additive
    rotationSpace=BCS_BoneSpace
    alphaBoolBlend {blendOption=Linear}
    boneToModify {boneName=head_offset}
    rotation {roll=-40}
   }
   n6 = "Transform (Modify) Bone - Bone: dyn_armor_r"(ComponentPose=@n5, Alpha=Get arm_back() -> (arm_back:double), Translation="-0.639,1.75,2.091", Rotation="15,8,-10", Scale="1,1,1") {
    rotationMode=BMM_Additive
    rotationSpace=BCS_BoneSpace
    translationMode=BMM_Additive
    translationSpace=BCS_BoneSpace
    alphaBoolBlend {blendOption=Linear}
    alphaScaleBiasClamp {bInterpResult=true, interpSpeedDecreasing=22, interpSpeedIncreasing=22}
    boneToModify {boneName=dyn_armor_r}
    rotation=<rot 15 8 -10>
    translation=<-0.639 1.75 2.091>
   }
   n7 = RigidBody(ComponentPose=@n6) {
    bClampLinearTranslationLimitToRefPose=true
    bOverrideWorldGravity=true
    overridePhysicsAsset=&/M_LRG_Invader_dyn_tube
    simulationSpace=BaseBoneSpace
    alphaBoolBlend {blendOption=Linear}
    baseBoneRef {boneName=spine_05}
    componentAppliedLinearAccClamp=<3000 3000 3000>
    componentLinearAccScale=<0.15 0.15 0.15>
    componentLinearVelScale=<0.2 0.2 0.2>
    overrideWorldGravity {z=-500}
   }
   Output Pose(Result=Component To Local(ComponentPose=@n7)) {layerGroup=None}
   Sequence Player 'invader_pickaxe'() {
    bLoopAnimation=false
    playRate=0.1
    playRateBasis=100
    sequence=/Game/Characters/Player/Male/Large/Bodies/M_LRG_Invader/Meshes/invader_pickaxe
    startPosition=11
   }
  ]}
 EventGraph {nodes [
   Event BlueprintInitializeAnimation()
   > Set SkelMesh(SkelMesh=GetPartSkeletalMeshComponent())
   K2Node_Event_2 = CustomEvent() {is=title K2Node_CustomEvent}
   n1 = Get Dynamic Material Instance(execute=[@K2Node_Event_2.then, @n34.then], SkeletalMeshComponent=GetPartSkeletalMeshComponent())
   > n2 = Sequence()
   n3 = PickaxeItemDefinitionHasAnyMetaTag(Tags=Make Array("[0]"=(TagName="Cosmetics.MetaTags.Invader")), bRequiresEquipped=true)
   > n4 = Branch(Condition=@n3)
   > n5 = Branch(Condition=Get bIsSkyDiving())
   SetScalarParameterValue(execute=[@n7.then, @n5.then, @n6.else, @n4.else], self=@n1, ParameterName=Hide Element 01, Value=0)
   SetScalarParameterValue(execute=[@n5.else, @n6.then], self=@n1, ParameterName=Hide Element 01, Value=1)
   n6 = Branch(execute=[@n8.then, @n2.then_1], Condition=@n3)
   n7 = Event OnBeginSkydiving()
   n8 = Event OnEndSkydiving()
   Event BlueprintUpdateAnimation()
   > Get!(execute:Exec) -> (then:Exec, else:Exec, SkelMesh:Skeletal Mesh Component Object Reference)
   > n9 = Sequence!(execute) -> (then_0, then_1, then_2:Exec, then_3:Exec, then_4:Exec, then_5:Exec, then_6:Exec, then_7:Exec, then_8:Exec, then_9:Exec)
   n10 = Get bIsParachuteOpened()
   n11 = Get bIsSkydiveDiveMode()
   n12 = Branch(execute=@n9.then_2, Condition=OR Boolean(A=OR Boolean(A=@n10, B=@n11), B=Get bIsSkyDiving()))
   > Set skydive_active(skydive_active:bool=true) -> (Output_Get:bool)
   Set skydive_active(execute=@n12.else, skydive_active=false)
   Set look_left(
    execute=@n9.then_5
    look_left:double=GetAlphaFromReferenceTransformDelta(
     SkeletalMeshComponent=Get SkelMesh()
     BoneName=head
     BaseName=spine_05
     DeltaComponentType=QuaternionTwist
     InputRange=(X=0,Y=45)
     TwistAxis="1,0,0"
    )
   ) -> (Output_Get:double)
   Set look_right(
    execute=@n9.then_6
    look_right:double=GetAlphaFromReferenceTransformDelta(
     SkeletalMeshComponent=Get SkelMesh()
     BoneName=head
     BaseName=neck_01
     DeltaComponentType=QuaternionTwist
     InputRange=(X=0,Y=-45)
     TwistAxis="1,0,0"
    )
   ) -> (Output_Get:double)
   Set FortPawn(execute=@n9.then_7, FortPawn:Fort Pawn Object Reference=TryGetFortPawn()) -> (Output_Get:Fort Pawn Object Reference)
   n13 = Get FortPawn() -> (FortPawn:Fort Pawn Object Reference)
   > n14 = Is Valid(InputObject=@n13)
   n15 = Set weapon_visible(execute=@n14.Is Not Valid, weapon_visible:bool=false) -> (Output_Get:bool)
   n16 = Get weapon_visible() -> (weapon_visible:bool)
   n17 = Branch(execute=[@n25.then, @n15.then], Condition=@n16)
   n18 = Get SkelMesh()
   > n19 = Get Dynamic Material Instance(SkeletalMeshComponent=@n18)
   n20 = Make Array("[0]"=(TagName="Cosmetics.MetaTags.Invader"))
   > n21 = Branch(Condition=AND Boolean(A=@n16, B=PickaxeItemDefinitionHasAnyMetaTag(Tags=@n20, bRequiresEquipped=true)))
   > SetScalarParameterValue(self=@n19, ParameterName=Hide Element 01, Value=1)
   > Set specific_weapon_visible(specific_weapon_visible:bool=true) -> (Output_Get:bool)
   > Set ctrl_rig_alpha(ctrl_rig_alpha:double=0) -> (Output_Get:double)
   n22 = Get Dynamic Material Instance(execute=@n17.else, SkeletalMeshComponent=@n18)
   SetScalarParameterValue(execute=[@n22.then, @n21.else], self=[@n22, @n19], ParameterName=Hide Element 01, Value=0)
   > Set specific_weapon_visible(specific_weapon_visible=false)
   n23 = Get CurrentWeapon!(self:Fort Pawn Object Reference=@n13) -> (CurrentWeapon:Fort Weapon Object Reference)
   n24 = IsWeaponHolstered!(self:Fort Pawn Object Reference=@n13) -> (ReturnValue:bool)
   n25 = Set weapon_visible(execute=@n14.Is Valid, weapon_visible=AND Boolean(A=IsValid(Object=@n23), B=NOT Boolean(A=@n24)))
   n26 = PSD_Func!(
    execute=@n9.then_8
    self
    startJnt:String=upperarm_r
    driverJnt:String=lowerarm_r
    poseJnt:String=pose_pickaxe_bck
    angle:double=63
    debug:bool=false
   ) -> (then, alpha:double)
   > Set arm_back(arm_back:double=@n26) -> (Output_Get:double)
   > Branch(Condition=Get specific_weapon_visible() -> (specific_weapon_visible:bool))
   > Set arm_back(arm_back=0)
   n27 = Branch(execute=@n9.then_9, Condition=Get specific_weapon_visible())
   > Set ctrl_rig_alpha(ctrl_rig_alpha=0)
   Set ctrl_rig_alpha(execute=@n27.else, ctrl_rig_alpha=1)
   n28 = Get SkelMesh()
   n29 = BreakRotator(InRot=BreakTransform(InTransform=GetDeltaTransformFromRefPose(self=@n28, BoneName=spine_05, BaseName=spine_01)).Rotation)
   Set spine_twist(execute=@n9.then_0, spine_twist:double=MapRangeClamped(Value=@n29, InRangeA=0, InRangeB=50)) -> (Output_Get:double)
   > Set spine_rot_bk(spine_rot_bk:double=MapRangeClamped(Value=@n29.Yaw, InRangeA=0, InRangeB=-30)) -> (Output_Get:double)
   n30 = Cast To FortPlayerAnimInstance(execute=@n9.then_4, Object=GetAnimInstance(self=GetOwnerSkeletalMeshComponent()))
   > Set weapon(weapon:String=Enum to String!(Enumerator:EFortWeaponCoreAnimation Enum=Get WeaponCoreAnim(self=@n30)) -> (ReturnValue:String)) -> (Output_Get:String)
   > n31 = Branch(Condition="Equal Exactly (String)"(A=Get weapon() -> (weapon:String), B=Pistol))
   > Set pistolAimAlpha(pistolAimAlpha:double=1) -> (Output_Get:double)
   Set pistolAimAlpha(execute=@n31.else, pistolAimAlpha=0)
   n32 = Get bIsSkyDiving()
   Set isSkydivingINV(execute=@n9.then_3, isSkydivingINV:double=float - float(B="To Float (Boolean)"(InBool=@n32)) {is=title K2Node_CallFunction}) -> (Output_Get:double)
   > Set isSkydiving(isSkydiving:double="To Float (Boolean)"(InBool=@n32)) -> (Output_Get:double)
   n33 = Get SkelMesh()
   Set arm_fwd_r(
    execute=@n9.then_1
    arm_fwd_r:double=GetAlphaFromReferenceTransformDelta(SkeletalMeshComponent=@n33, BoneName=upperarm_r, BaseName=spine_05, InputRange=(X=60,Y=150), TwistAxis="0,0,0")
   ) -> (Output_Get:double)
   > Set clavicle_up_r(clavicle_up_r:double=GetAlphaFromReferenceTransformDelta(SkeletalMeshComponent=@n33, BoneName=clavicle_r, BaseName=spine_05, DeltaComponentType=EulerY, InputRange=(X=30,Y=40), TwistAxis="0,0,0")) -> (Output_Get:double)
   > Branch(Condition=Get skydive_active() -> (skydive_active:bool))
   > Set arm_fwd_r(arm_fwd_r=0)
   n34 = Event OnCharacterCustomizationLoaded()
   PrintString!(execute, InString="To String (Float)"(InDouble=0), bPrintToScreen, bPrintToLog, TextColor, Duration, Key) -> (then)
   GetAlphaFromReferenceTransformDelta(BoneName=None, BaseName=None, DeltaComponentType=QuaternionTwist, InputRange=(X=0,Y=1), TwistAxis="1,0,0")
   Get bIsSkydiveFloating()
   Set clavicle_up_r(clavicle_up_r=0)
  ]}
 PSD_Func {nodes [
   n1 = PSD_Func!() -> (then, startJnt:String, driverJnt:String, poseJnt:String, angle:double=~, debug:bool=~) {is=title K2Node_FunctionEntry}
   > n2 = Sequence!(execute) -> (then_0, then_1)
   > Branch(Condition=@n1.debug)
   n3 = GetPartSkeletalMeshComponent()
   n4 = GetSocketLocation!(self:Scene Component Object Reference=@n3, InSocketName:Name=String To Name(InString=@n1)) -> (ReturnValue:Vector)
   > DrawDebugArrow!(
    execute
    LineStart:Vector=@n4
    LineEnd:Vector=GetSocketLocation(self=@n3, InSocketName=String To Name(InString=@n1.poseJnt))
    ArrowSize:float=0
    LineColor:Linear Color Structure=(R=0.1692,G=0.147,B=1,A=1)
    Duration:float=0
    Thickness:float=1
    DepthPriority:EDrawDebugSceneDepthPriorityGroup Enum=World
   ) -> (then)
   n5 = GetPartSkeletalMeshComponent()
   n6 = Get Direction Between Sockets!(
    Component:Skeletal Mesh Component Object Reference=@n5
    SocketOrBoneNameFrom:Name=String To Name(InString=@n1)
    SocketOrBoneNameTo:Name=String To Name(InString=@n1.poseJnt)
   ) -> (ReturnValue:Vector) {is=title K2Node_CallFunction}
   > Draw Debug Cone!(
    execute
    Origin:Vector=@n4
    Direction:Vector=@n6
    Length:float=100
    AngleWidth:float=@n1.angle
    AngleHeight:float=@n1.angle
    NumSides:int=12
    LineColor:Linear Color Structure=(R=1,G=1,B=1,A=1)
    Duration:float=0
    Thickness:float=0.3
    DepthPriority:EDrawDebugSceneDepthPriorityGroup Enum=World
   ) -> (then)
   n7 = Get Direction Between Sockets(Component=@n5, SocketOrBoneNameFrom=String To Name(InString=@n1), SocketOrBoneNameTo=String To Name(InString=@n1.driverJnt))
   n8 = float - float(B=float / float(A=GetAngleBetweenTwoVectors!(VectorA:Vector=@n7, VectorB:Vector=@n6) -> ("Angle (Degrees)":double), B=@n1.angle) {is=title K2Node_CallFunction})
   n9 = Branch(execute=@n2.then_1, Condition="float >= float"(A=@n8, B=0) {is=title K2Node_CallFunction})
   > n10 = Set psd_alpha(psd_alpha:double=@n8) -> (Output_Get:double)
   Return Node!(execute=[@n11.then, @n10.then], alpha:double=@n10)
   n11 = Set psd_alpha(execute=@n9.else, psd_alpha=0)
  ]}
]
variables [
 arm_back {type=number}
 arm_fwd_r {type=number}
 clavicle_up_r {type=number}
 ctrl_rig_alpha {type=number}
 FortPawn {type=FortPawn, default=null}
 isCrouching {type=number}
 isSkydiving {type=number}
 isSkydivingINV {type=number}
 look_left {type=number}
 look_right {type=number}
 pistolAimAlpha {type=number}
 psd_alpha {type=number}
 SkelMesh {type=SkeletalMeshComponent}
 skydive_active {type=boolean}
 specific_weapon_visible {type=boolean}
 spine_rot_bk {type=number}
 spine_twist {type=number}
 weapon {type=string, default=""}
 weapon_visible {type=boolean}
]
