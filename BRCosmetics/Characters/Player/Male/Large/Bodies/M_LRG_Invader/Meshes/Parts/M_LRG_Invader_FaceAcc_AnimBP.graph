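# Animation Blueprint for the Invader large male character's face accessory mesh, rolling its head_offset bone as the head turns left or right.
# Each update derives look_left and look_right alphas from head bone twist on the part's skeletal mesh component, relative to spine_05 and neck_01 respectively.
# The AnimGraph copies the attached parent's pose, then applies two opposing additive bone-space rotations weighted by those alphas.
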AnimBlueprint /BRCosmetics/Characters/Player/Male/Large/Bodies/M_LRG_Invader/Meshes/Parts/M_LRG_Invader_FaceAcc_AnimBP
target_skeleton=&/M_LRG_Invader_FaceAcc_Skeleton
graphs [
 AnimGraph {nodes [
   n1 = "Transform (Modify) Bone - Bone: head_offset"(
    ComponentPose=Local To Component(LocalPose=Copy Pose From Mesh() {bUseAttachedParent=true})
    Alpha=Get look_left() -> (look_left:double)
    Translation="0,0,0"
    Rotation="0,0,40"
    Scale="1,1,1"
   ) {
    rotationMode=BMM_Additive
    rotationSpace=BCS_BoneSpace
    alphaBoolBlend {blendOption=Linear}
    boneToModify {boneName=head_offset}
    rotation {roll=40}
   }
   n2 = "Transform (Modify) Bone - Bone: head_offset"(ComponentPose=@n1, Alpha=Get look_right() -> (look_right:double), Translation="0,0,0", Rotation="0,0,-40", Scale="1,1,1") {
    rotationMode=BMM_Additive
    rotationSpace=BCS_BoneSpace
    alphaBoolBlend {blendOption=Linear}
    boneToModify {boneName=head_offset}
    rotation {roll=-40}
   }
   Output Pose(Result=Component To Local(ComponentPose=@n2)) {layerGroup=None}
  ]}
 EventGraph {nodes [
   Event BlueprintUpdateAnimation()
   > Get!(execute:Exec) -> (then:Exec, else:Exec, SkelMesh:Skeletal Mesh Component Object Reference)
   > n1 = Sequence()
   > Set look_left(look_left:double=GetAlphaFromReferenceTransformDelta(
     SkeletalMeshComponent=Get SkelMesh()
     BoneName=head
     BaseName=spine_05
     DeltaComponentType=QuaternionTwist
     InputRange=(X=0,Y=45)
     TwistAxis="1,0,0"
    )) -> (Output_Get:double)
   Set look_right(
    execute=@n1.then_1
    look_right:double=GetAlphaFromReferenceTransformDelta(
     SkeletalMeshComponent=Get SkelMesh()
     BoneName=head
     BaseName=neck_01
     DeltaComponentType=QuaternionTwist
     InputRange=(X=0,Y=-45)
     TwistAxis="1,0,0"
    )
   ) -> (Output_Get:double)
   Event BlueprintInitializeAnimation()
   > Set SkelMesh(SkelMesh=GetPartSkeletalMeshComponent())
  ]}
]
variables [
 look_left {type=number}
 look_right {type=number}
 SkelMesh {type=SkeletalMeshComponent}
]
